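# Sound cue for the landing foley of the snowy fish cosmetic companion.
# A quality-level switch sends its first two slots to a random pick of six land waves and its last two to a random pick of the first four, with slight pitch variation on top.
# It uses companion foley attenuation, the pets sound class and the concurrency settings of quadruped footsteps, at zero priority.
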SoundCue /CosmeticCompanions/Assets/Audio_Shared/Foley/Other/Fish_Snowy/SC_Mimosa_Other_Fish_Snowy_Land
root=Modulator_0
nodes [
 {id=Modulator_0, children [QualityLevel_0], properties {pitchMax=1.1, pitchMin=0.9, volumeMax=1, volumeMin=1}}
 {id=QualityLevel_0, children [Random_2,Random_2,Random_3,Random_3]}
 {
  id=Random_2
  children [WavePlayer_2, WavePlayer_3, WavePlayer_10, WavePlayer_11, WavePlayer_12, WavePlayer_13]
  properties {weights [1,1,1,1,1,1]}
 }
 {id=Random_3, children [WavePlayer_14, WavePlayer_15, WavePlayer_16, WavePlayer_17], properties {weights [1,1,1,1]}}
 {
  id=WavePlayer_10
  properties {soundWaveAssetPtr=/CosmeticCompanions/Assets/Audio_Shared/Foley/Other/Fish_Snowy/Wavs/Mimosa_Fish_Snowy_Land_03}
 }
 {
  id=WavePlayer_11
  properties {soundWaveAssetPtr=/CosmeticCompanions/Assets/Audio_Shared/Foley/Other/Fish_Snowy/Wavs/Mimosa_Fish_Snowy_Land_04}
 }
 {
  id=WavePlayer_12
  properties {soundWaveAssetPtr=/CosmeticCompanions/Assets/Audio_Shared/Foley/Other/Fish_Snowy/Wavs/Mimosa_Fish_Snowy_Land_05}
 }
 {
  id=WavePlayer_13
  properties {soundWaveAssetPtr=/CosmeticCompanions/Assets/Audio_Shared/Foley/Other/Fish_Snowy/Wavs/Mimosa_Fish_Snowy_Land_06}
 }
 {
  id=WavePlayer_14
  properties {soundWaveAssetPtr=/CosmeticCompanions/Assets/Audio_Shared/Foley/Other/Fish_Snowy/Wavs/Mimosa_Fish_Snowy_Land_01}
 }
 {
  id=WavePlayer_15
  properties {soundWaveAssetPtr=/CosmeticCompanions/Assets/Audio_Shared/Foley/Other/Fish_Snowy/Wavs/Mimosa_Fish_Snowy_Land_02}
 }
 {
  id=WavePlayer_16
  properties {soundWaveAssetPtr=&Mimosa_Fish_Snowy_Land_03}
 }
 {
  id=WavePlayer_17
  properties {soundWaveAssetPtr=&Mimosa_Fish_Snowy_Land_04}
 }
 {
  id=WavePlayer_2
  properties {soundWaveAssetPtr=&Mimosa_Fish_Snowy_Land_01}
 }
 {
  id=WavePlayer_3
  properties {soundWaveAssetPtr=&Mimosa_Fish_Snowy_Land_02}
 }
]
properties {
 attenuationSettings=/Game/Sounds/Attenuation/AS_Companion_Foley
 maxDistance=1200
 priority=0
 soundClassObject=&Fort_Creatures_Pets
 concurrencySet [/CosmeticCompanions/Assets/Audio_Shared/Foley/Quadraped/CONC_Mimosa_Quadraped_Footstep]
}
