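# Sound cue for the spin foley of the Fish_Snowy cosmetic companion, played in the pet creature sound class.
# A quality-level node routes its first two levels to a random pick of five spin waves and its last two to a random pick of only the first four.
# It uses companion foley attenuation and a boosted volume, and shares the quadruped footstep concurrency set.
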SoundCue /CosmeticCompanions/Assets/Audio_Shared/Foley/Other/Fish_Snowy/SC_Mimosa_Other_Fish_Snowy_Spin
root=QualityLevel_0
nodes [
 {id=QualityLevel_0, children [Random_0,Random_0,Random_3,Random_3]}
 {id=Random_0, children [WavePlayer_1,WavePlayer_2,WavePlayer_3,WavePlayer_4,WavePlayer_5], properties {weights [1,1,1,1,1]}}
 {id=Random_3, children [WavePlayer_6,WavePlayer_7,WavePlayer_8,WavePlayer_9], properties {weights [1,1,1,1]}}
 {
  id=WavePlayer_1
  properties {soundWaveAssetPtr=/CosmeticCompanions/Assets/Audio_Shared/Foley/Other/Fish_Snowy/Wavs/Mimosa_Fish_Snowy_Spin_01}
 }
 {
  id=WavePlayer_2
  properties {soundWaveAssetPtr=/CosmeticCompanions/Assets/Audio_Shared/Foley/Other/Fish_Snowy/Wavs/Mimosa_Fish_Snowy_Spin_02}
 }
 {
  id=WavePlayer_3
  properties {soundWaveAssetPtr=/CosmeticCompanions/Assets/Audio_Shared/Foley/Other/Fish_Snowy/Wavs/Mimosa_Fish_Snowy_Spin_03}
 }
 {
  id=WavePlayer_4
  properties {soundWaveAssetPtr=/CosmeticCompanions/Assets/Audio_Shared/Foley/Other/Fish_Snowy/Wavs/Mimosa_Fish_Snowy_Spin_04}
 }
 {
  id=WavePlayer_5
  properties {soundWaveAssetPtr=/CosmeticCompanions/Assets/Audio_Shared/Foley/Other/Fish_Snowy/Wavs/Mimosa_Fish_Snowy_Spin_05}
 }
 {
  id=WavePlayer_6
  properties {soundWaveAssetPtr=&Mimosa_Fish_Snowy_Spin_01}
 }
 {
  id=WavePlayer_7
  properties {soundWaveAssetPtr=&Mimosa_Fish_Snowy_Spin_02}
 }
 {
  id=WavePlayer_8
  properties {soundWaveAssetPtr=&Mimosa_Fish_Snowy_Spin_03}
 }
 {
  id=WavePlayer_9
  properties {soundWaveAssetPtr=&Mimosa_Fish_Snowy_Spin_04}
 }
]
properties {
 attenuationSettings=/Game/Sounds/Attenuation/AS_Companion_Foley
 maxDistance=1200
 priority=0
 soundClassObject=&Fort_Creatures_Pets
 volumeMultiplier=1.5
 concurrencySet [/CosmeticCompanions/Assets/Audio_Shared/Foley/Quadraped/CONC_Mimosa_Quadraped_Footstep]
}
