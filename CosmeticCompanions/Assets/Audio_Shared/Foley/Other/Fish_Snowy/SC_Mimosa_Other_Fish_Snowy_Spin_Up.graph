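# Sound cue for the spin-up foley of the Fish_Snowy cosmetic companion.
# A root Random node picks one of three equally weighted wave variations.
# Playback uses companion foley attenuation and the creature pets sound class, at priority zero.
# Its concurrency is the shared quadruped footstep set rather than one specific to the fish.
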SoundCue /CosmeticCompanions/Assets/Audio_Shared/Foley/Other/Fish_Snowy/SC_Mimosa_Other_Fish_Snowy_Spin_Up
root=Random_3
nodes [
 {id=Random_3, children [WavePlayer_1,WavePlayer_2,WavePlayer_3], properties {weights [1,1,1]}}
 {
  id=WavePlayer_1
  properties {soundWaveAssetPtr=/CosmeticCompanions/Assets/Audio_Shared/Foley/Other/Fish_Snowy/Wavs/Mimosa_Fish_Snowy_Spin_Up_01}
 }
 {
  id=WavePlayer_2
  properties {soundWaveAssetPtr=/CosmeticCompanions/Assets/Audio_Shared/Foley/Other/Fish_Snowy/Wavs/Mimosa_Fish_Snowy_Spin_Up_02}
 }
 {
  id=WavePlayer_3
  properties {soundWaveAssetPtr=/CosmeticCompanions/Assets/Audio_Shared/Foley/Other/Fish_Snowy/Wavs/Mimosa_Fish_Snowy_Spin_Up_03}
 }
]
properties {
 attenuationSettings=/Game/Sounds/Attenuation/AS_Companion_Foley
 maxDistance=1200
 priority=0
 soundClassObject=&Fort_Creatures_Pets
 volumeMultiplier=1
 concurrencySet [/CosmeticCompanions/Assets/Audio_Shared/Foley/Quadraped/CONC_Mimosa_Quadraped_Footstep]
}
