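# Sound Cue for the turn foley of the Fish Snowy cosmetic companion.
# A modulator randomizes pitch and volume over a quality-level switch feeding random wave selection.
# The first two quality slots draw from six turn variations, while the other two use only the first four.
# It plays in the pets sound class with companion foley attenuation and borrows the quadruped footstep concurrency set.
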SoundCue /CosmeticCompanions/Assets/Audio_Shared/Foley/Other/Fish_Snowy/SC_Mimosa_Other_Fish_Snowy_Turn
root=Modulator_0
nodes [
 {id=Modulator_0, children [QualityLevel_0], properties {pitchMax=1.1, pitchMin=0.8, volumeMax=1.1, volumeMin=0.7}}
 {id=QualityLevel_0, children [Random_3,Random_3,Random_5,Random_5]}
 {
  id=Random_3
  children [WavePlayer_8, WavePlayer_9, WavePlayer_10, WavePlayer_11, WavePlayer_12, WavePlayer_13]
  properties {weights [1,1,1,1,1,1]}
 }
 {id=Random_5, children [WavePlayer_14, WavePlayer_15, WavePlayer_16, WavePlayer_17], properties {weights [1,1,1,1]}}
 {
  id=WavePlayer_10
  properties {soundWaveAssetPtr=/CosmeticCompanions/Assets/Audio_Shared/Foley/Other/Fish_Snowy/Wavs/Mimosa_Fish_Snowy_Turn_03}
 }
 {
  id=WavePlayer_11
  properties {soundWaveAssetPtr=/CosmeticCompanions/Assets/Audio_Shared/Foley/Other/Fish_Snowy/Wavs/Mimosa_Fish_Snowy_Turn_04}
 }
 {
  id=WavePlayer_12
  properties {soundWaveAssetPtr=/CosmeticCompanions/Assets/Audio_Shared/Foley/Other/Fish_Snowy/Wavs/Mimosa_Fish_Snowy_Turn_05}
 }
 {
  id=WavePlayer_13
  properties {soundWaveAssetPtr=/CosmeticCompanions/Assets/Audio_Shared/Foley/Other/Fish_Snowy/Wavs/Mimosa_Fish_Snowy_Turn_06}
 }
 {
  id=WavePlayer_14
  properties {soundWaveAssetPtr=/CosmeticCompanions/Assets/Audio_Shared/Foley/Other/Fish_Snowy/Wavs/Mimosa_Fish_Snowy_Turn_01}
 }
 {
  id=WavePlayer_15
  properties {soundWaveAssetPtr=/CosmeticCompanions/Assets/Audio_Shared/Foley/Other/Fish_Snowy/Wavs/Mimosa_Fish_Snowy_Turn_02}
 }
 {
  id=WavePlayer_16
  properties {soundWaveAssetPtr=&Mimosa_Fish_Snowy_Turn_03}
 }
 {
  id=WavePlayer_17
  properties {soundWaveAssetPtr=&Mimosa_Fish_Snowy_Turn_04}
 }
 {
  id=WavePlayer_8
  properties {soundWaveAssetPtr=&Mimosa_Fish_Snowy_Turn_01}
 }
 {
  id=WavePlayer_9
  properties {soundWaveAssetPtr=&Mimosa_Fish_Snowy_Turn_02}
 }
]
properties {
 attenuationSettings=/Game/Sounds/Attenuation/AS_Companion_Foley
 maxDistance=1200
 priority=0
 soundClassObject=&Fort_Creatures_Pets
 volumeMultiplier=0.6
 concurrencySet [/CosmeticCompanions/Assets/Audio_Shared/Foley/Quadraped/CONC_Mimosa_Quadraped_Footstep]
}
